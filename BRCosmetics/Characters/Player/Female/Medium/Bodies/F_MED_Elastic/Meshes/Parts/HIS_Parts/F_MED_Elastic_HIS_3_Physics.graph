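# Physics asset for the HIS_3 hair part of the F_MED_Elastic female player cosmetic, defining capsule bodies for the hair and upper body.
# Capsules cover head, neck, upper spine, clavicles and upper arms, plus back, side, front, top and split hair bones.
# Hair chains hang from base bodies under his3_hair with limited swing and twist, tightest on the twist-locked front strand.
# Extra cross constraints tie neighboring chains together, and link head and neck to the right A chain.
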PhysicsAsset /BRCosmetics/Characters/Player/Female/Medium/Bodies/F_MED_Elastic/Meshes/Parts/HIS_Parts/F_MED_Elastic_HIS_3_Physics
bodies {
 clavicle_l [{shapeType=Capsule, length=10.5, radius=5.25, center=<4.2923 -0.9803 -0.1045>, rotation=<rot 78.3542 0 0>}]
 clavicle_r [{shapeType=Capsule, length=10.5, radius=5.25, center=<-4.2923 0.9803 0.1045>, rotation=<rot 78.3542 0 -180>}]
 head [{shapeType=Capsule, length=5.25, radius=6, center=<3.2039 2.8659 0>, rotation=<rot 90 0 0>}]
 his3_hair [{shapeType=Capsule, length=1, radius=1}]
 his3_hair_A_a_base_l [{shapeType=Capsule, length=1, radius=1}]
 his3_hair_A_a_base_r [{shapeType=Capsule, length=1, radius=1}]
 his3_hair_A_a_l [{shapeType=Capsule, length=1.75, radius=2.25, center=<2.71 0.2771 0>, rotation=<rot 90 17.354 11.5164>}]
 his3_hair_A_a_r [{shapeType=Capsule, length=3.25, radius=1.75, center=<2.7812 0.0001 0>, rotation=<rot 90 0 0>}]
 his3_hair_A_b_l [{shapeType=Capsule, length=1.5, radius=3, center=<2.9805 -0.0001 -0.2398>, rotation=<rot 85.3992 -0.0001 -0.0001>}]
 his3_hair_A_b_r [{shapeType=Capsule, length=3, radius=2.25, center=<3.5053 0 0>, rotation=<rot 88.5595 0.0018 0.0018>}]
 his3_hair_A_c_l [{shapeType=Capsule, length=1.25, radius=2.75, center=<2.1729 -0.6847 0.6815>, rotation=<rot 76.9589 -5.6257 -0.0003>}]
 his3_hair_A_c_r [{shapeType=Capsule, length=0.75, radius=2.75, center=<1.8778 0 0>, rotation=<rot 90 0 0>}]
 his3_hair_A_d_l [{shapeType=Capsule, length=2.5, radius=1.5, center=<2.1391 0 -0.4018>, rotation=<rot 83.2109 0.0002 0.0002>}]
 his3_hair_A_d_r [{shapeType=Capsule, length=1.5, radius=1.75, center=<2.3195 0 0>, rotation=<rot 90 -0.7162 -4.5031>}]
 his3_hair_A_e_r [{shapeType=Capsule, length=3, radius=1.25, center=<2.536 0 0>, rotation=<rot 90 0 -5.6063>}]
 his3_hair_A_small_a_l [{shapeType=Capsule, length=1.75, radius=0.75, center=<0.9621 0 0.5956>, rotation=<rot 58.24 180 180>}]
 his3_hair_A_small_a_r [{shapeType=Capsule, length=2, radius=1.25, center=<1.8381 0 0>, rotation=<rot 90 0 0>}]
 his3_hair_A_small_b_l [{shapeType=Capsule, length=1.75, radius=1, center=<1.6768 0 0>, rotation=<rot 90 0 0>}]
 his3_hair_A_small_b_r [{shapeType=Capsule, length=2.5, radius=0.75, center=<1.5912 0 0>, rotation=<rot 90 0 0>}]
 his3_hair_B_a_base_l [{shapeType=Capsule, length=1, radius=1}]
 his3_hair_B_a_base_r [{shapeType=Capsule, length=1, radius=1}]
 his3_hair_B_a_l [{shapeType=Capsule, length=0.5, radius=3.5, center=<2.7762 0 0>, rotation=<rot 90 0 0>}]
 his3_hair_B_a_r [{shapeType=Capsule, length=2, radius=2.5, center=<2.8219 0 0>, rotation=<rot 90 0 0>}]
 his3_hair_B_b_l [{shapeType=Capsule, length=1, radius=3.5, center=<2.7426 -0.4368 0>, rotation=<rot 90 24.6235 33.6724>}]
 his3_hair_B_b_r [{shapeType=Capsule, length=3.5, radius=2.25, center=<3.4416 0 0>, rotation=<rot 90 0 0>}]
 his3_hair_B_c_l [{shapeType=Capsule, length=1.25, radius=2.75, center=<2.5242 -0.0001 -0.4095>, rotation=<rot 80.784 -0.0001 -0.0001>}]
 his3_hair_B_c_r [{shapeType=Capsule, length=1.25, radius=2.75, center=<2.3992 0 -0.4658>, rotation=<rot 76.9317 0.0002 0.0002>}]
 his3_hair_B_d_l [{shapeType=Capsule, length=3.5, radius=1.5, center=<3.6676 -0.2119 0>, rotation=<rot 90 -18.9704 -16.2751>}]
 his3_hair_B_d_r [{shapeType=Capsule, length=2, radius=1.75, center=<2.3455 -0.351 -0.0001>, rotation=<rot 90 0.8952 3.0874>}]
 his3_hair_bk_a [{shapeType=Capsule, length=1.5, radius=4.25, center=<3.4115 0 0>, rotation=<rot 90 0 0>}]
 his3_hair_bk_a_base [{shapeType=Capsule, length=1, radius=1}]
 his3_hair_bk_b [{shapeType=Capsule, length=2, radius=3.5, center=<3.0021 0 0>, rotation=<rot 90 0 8.2087>}]
 his3_hair_bk_c [{shapeType=Capsule, length=1.5, radius=3, center=<2.9538 0 0>, rotation=<rot 90 0 0>}]
 his3_hair_bk_d [{shapeType=Capsule, length=4.5, radius=1.5, center=<2.907 -0.1233 0.7102>, rotation=<rot 74.5457 175.9885 -176.453>}]
 his3_hair_fr_A_a_base_r [{shapeType=Capsule, length=1, radius=1}]
 his3_hair_fr_A_a_r [{shapeType=Capsule, length=2.5, radius=1, center=<2.1129 0 0>, rotation=<rot 90 0 0>}]
 his3_hair_fr_A_b_r [{shapeType=Capsule, length=1.75, radius=1, center=<1.6382 -0.3913 0>, rotation=<rot 90 26.565 48.0726>}]
 his3_hair_fr_A_c_r [{shapeType=Capsule, length=3, radius=1, center=<2.1803 0 0>, rotation=<rot 90 0 0>}]
 his3_hair_fr_A_d_r [{shapeType=Capsule, length=4, radius=1, center=<3.0394 0 0>, rotation=<rot 90 0 0>}]
 his3_hair_split_base_r [{shapeType=Capsule, length=0.505, radius=0.505, rotation=<rot 0 -90 0>}]
 his3_hair_split_r [{shapeType=Capsule, length=3.4749, radius=2.7929, center=<3.1319 -1.1024 -0.3733>, rotation=<rot 1.1201 -107.9058 -75.6179>}]
 his3_hair_top_A_r [{shapeType=Capsule, length=2.0157, radius=0.7934, center=<0.7776 -0.2097 -1.5368>, rotation=<rot 10.7431 -90 -23.9539>}]
 neck_01 [{shapeType=Capsule, length=3, radius=4.25, center=<1.4608 -0.2332 0>, rotation=<rot 90 0 -4.3894>}]
 spine_04 [{shapeType=Capsule, length=5.5, radius=10, center=<1.6154 2.4791 0>}]
 spine_05 [{shapeType=Capsule, length=13.25, radius=7.75, center=<1.8697 -0.4876 0>}]
 upperarm_l [{shapeType=Capsule, length=20.5, radius=4, center=<11.8956 -0.5824 0.0001>, rotation=<rot 90 0.8952 4.0065>}]
 upperarm_r [{shapeType=Capsule, length=20.5, radius=4, center=<-11.8956 0.5824 -0.0001>, rotation=<rot 90 -179.1048 4.0065>}]
}
constraints [
 {
  bone1Name=head
  bone2Name=neck_01
  swing1LimitDegrees=45
  swing1Motion=Free
  swing2LimitDegrees=45
  swing2Motion=Free
  twistLimitDegrees=45
  twistMotion=Free
 }
 {
  bone1Name=neck_01
  bone2Name=spine_05
  swing1LimitDegrees=45
  swing1Motion=Free
  swing2LimitDegrees=45
  swing2Motion=Free
  twistLimitDegrees=45
  twistMotion=Free
 }
 {
  bone1Name=clavicle_l
  bone2Name=spine_05
  swing1LimitDegrees=45
  swing1Motion=Free
  swing2LimitDegrees=45
  swing2Motion=Free
  twistLimitDegrees=45
  twistMotion=Free
 }
 {
  bone1Name=clavicle_r
  bone2Name=spine_05
  swing1LimitDegrees=45
  swing1Motion=Free
  swing2LimitDegrees=45
  swing2Motion=Free
  twistLimitDegrees=45
  twistMotion=Free
 }
 {
  bone1Name=spine_05
  bone2Name=spine_04
  swing1LimitDegrees=45
  swing1Motion=Free
  swing2LimitDegrees=45
  swing2Motion=Free
  twistLimitDegrees=45
  twistMotion=Free
 }
 {
  bone1Name=upperarm_l
  bone2Name=clavicle_l
  swing1LimitDegrees=45
  swing1Motion=Free
  swing2LimitDegrees=45
  swing2Motion=Free
  twistLimitDegrees=45
  twistMotion=Free
 }
 {
  bone1Name=upperarm_r
  bone2Name=clavicle_r
  swing1LimitDegrees=45
  swing1Motion=Free
  swing2LimitDegrees=45
  swing2Motion=Free
  twistLimitDegrees=45
  twistMotion=Free
 }
 {
  bone1Name=his3_hair
  bone2Name=head
  swing1LimitDegrees=45
  swing1Motion=Free
  swing2LimitDegrees=45
  swing2Motion=Free
  twistLimitDegrees=45
  twistMotion=Free
 }
 {
  bone1Name=his3_hair_bk_a_base
  bone2Name=his3_hair
  swing1LimitDegrees=45
  swing1Motion=Free
  swing2LimitDegrees=45
  swing2Motion=Free
  twistLimitDegrees=45
  twistMotion=Free
 }
 {
  bone1Name=his3_hair_B_a_base_l
  bone2Name=his3_hair
  swing1LimitDegrees=45
  swing1Motion=Free
  swing2LimitDegrees=45
  swing2Motion=Free
  twistLimitDegrees=45
  twistMotion=Free
 }
 {
  bone1Name=his3_hair_A_a_base_l
  bone2Name=his3_hair
  swing1LimitDegrees=45
  swing1Motion=Free
  swing2LimitDegrees=45
  swing2Motion=Free
  twistLimitDegrees=45
  twistMotion=Free
 }
 {
  bone1Name=his3_hair_B_a_base_r
  bone2Name=his3_hair
  swing1LimitDegrees=45
  swing1Motion=Free
  swing2LimitDegrees=45
  swing2Motion=Free
  twistLimitDegrees=45
  twistMotion=Free
 }
 {
  bone1Name=his3_hair_A_a_base_r
  bone2Name=his3_hair
  swing1LimitDegrees=45
  swing1Motion=Free
  swing2LimitDegrees=45
  swing2Motion=Free
  twistLimitDegrees=45
  twistMotion=Free
 }
 {
  bone1Name=his3_hair_fr_A_a_base_r
  bone2Name=his3_hair
  swing1LimitDegrees=45
  swing1Motion=Free
  swing2LimitDegrees=45
  swing2Motion=Free
  twistLimitDegrees=45
  twistMotion=Free
 }
 {bone1Name=his3_hair_bk_a, bone2Name=his3_hair_bk_a_base, swing1LimitDegrees=30, swing2LimitDegrees=15, twistLimitDegrees=5}
 {bone1Name=his3_hair_bk_b, bone2Name=his3_hair_bk_a, swing1LimitDegrees=30, swing2LimitDegrees=15, twistLimitDegrees=5}
 {bone1Name=his3_hair_bk_c, bone2Name=his3_hair_bk_b, swing1LimitDegrees=30, swing2LimitDegrees=15, twistLimitDegrees=5}
 {bone1Name=his3_hair_bk_d, bone2Name=his3_hair_bk_c, swing1LimitDegrees=20, swing2LimitDegrees=15, twistLimitDegrees=5}
 {bone1Name=his3_hair_B_a_l, bone2Name=his3_hair_B_a_base_l, swing1LimitDegrees=25, swing2LimitDegrees=25, twistLimitDegrees=5}
 {bone1Name=his3_hair_B_b_l, bone2Name=his3_hair_B_a_l, swing1LimitDegrees=25, swing2LimitDegrees=25, twistLimitDegrees=5}
 {bone1Name=his3_hair_B_c_l, bone2Name=his3_hair_B_b_l, swing1LimitDegrees=25, swing2LimitDegrees=25, twistLimitDegrees=5}
 {bone1Name=his3_hair_B_d_l, bone2Name=his3_hair_B_c_l, swing1LimitDegrees=20, swing2LimitDegrees=20, twistLimitDegrees=5}
 {bone1Name=his3_hair_A_a_l, bone2Name=his3_hair_A_a_base_l, swing1LimitDegrees=20, swing2LimitDegrees=20, twistLimitDegrees=5}
 {bone1Name=his3_hair_A_b_l, bone2Name=his3_hair_A_a_l, swing1LimitDegrees=20, swing2LimitDegrees=20, twistLimitDegrees=5}
 {bone1Name=his3_hair_A_c_l, bone2Name=his3_hair_A_b_l, swing1LimitDegrees=20, swing2LimitDegrees=20, twistLimitDegrees=5}
 {bone1Name=his3_hair_A_d_l, bone2Name=his3_hair_A_c_l, swing1LimitDegrees=10, swing2LimitDegrees=10, twistLimitDegrees=5}
 {bone1Name=his3_hair_A_small_b_l, bone2Name=his3_hair_A_c_l, swing1LimitDegrees=15, swing2LimitDegrees=15, twistLimitDegrees=15}
 {bone1Name=his3_hair_A_small_a_l, bone2Name=his3_hair_A_b_l, swing1LimitDegrees=15, swing2LimitDegrees=15, twistLimitDegrees=15}
 {bone1Name=his3_hair_B_a_r, bone2Name=his3_hair_B_a_base_r, swing1LimitDegrees=25, swing2LimitDegrees=25, twistLimitDegrees=5}
 {bone1Name=his3_hair_B_b_r, bone2Name=his3_hair_B_a_r, swing1LimitDegrees=25, swing2LimitDegrees=25, twistLimitDegrees=5}
 {bone1Name=his3_hair_B_c_r, bone2Name=his3_hair_B_b_r, swing1LimitDegrees=25, swing2LimitDegrees=25, twistLimitDegrees=5}
 {bone1Name=his3_hair_B_d_r, bone2Name=his3_hair_B_c_r, swing1LimitDegrees=20, swing2LimitDegrees=20, twistLimitDegrees=5}
 {bone1Name=his3_hair_A_a_r, bone2Name=his3_hair_A_a_base_r, swing1LimitDegrees=15, swing2LimitDegrees=15, twistLimitDegrees=5}
 {bone1Name=his3_hair_A_b_r, bone2Name=his3_hair_A_a_r, swing1LimitDegrees=15, swing2LimitDegrees=15, twistLimitDegrees=5}
 {bone1Name=his3_hair_A_c_r, bone2Name=his3_hair_A_b_r, swing1LimitDegrees=15, swing2LimitDegrees=15, twistLimitDegrees=5}
 {bone1Name=his3_hair_A_d_r, bone2Name=his3_hair_A_c_r, swing1LimitDegrees=15, swing2LimitDegrees=15, twistLimitDegrees=5}
 {bone1Name=his3_hair_A_e_r, bone2Name=his3_hair_A_d_r, swing1LimitDegrees=15, swing2LimitDegrees=15, twistLimitDegrees=5}
 {bone1Name=his3_hair_A_small_a_r, bone2Name=his3_hair_A_c_r, swing1LimitDegrees=15, swing2LimitDegrees=15, twistLimitDegrees=5}
 {bone1Name=his3_hair_A_small_b_r, bone2Name=his3_hair_A_small_a_r, swing1LimitDegrees=15, swing2LimitDegrees=15, twistLimitDegrees=5}
 {bone1Name=his3_hair_top_A_r, bone2Name=his3_hair, swing1LimitDegrees=10, swing2LimitDegrees=10, twistLimitDegrees=10}
 {
  bone1Name=his3_hair_fr_A_a_r
  bone2Name=his3_hair_fr_A_a_base_r
  swing1LimitDegrees=2
  swing2LimitDegrees=2
  twistLimitDegrees=5
  twistMotion=Locked
 }
 {
  bone1Name=his3_hair_fr_A_b_r
  bone2Name=his3_hair_fr_A_a_r
  swing1LimitDegrees=2
  swing2LimitDegrees=2
  twistLimitDegrees=5
  twistMotion=Locked
 }
 {
  bone1Name=his3_hair_fr_A_d_r
  bone2Name=his3_hair_fr_A_b_r
  swing1LimitDegrees=2
  swing2LimitDegrees=2
  twistLimitDegrees=5
  twistMotion=Locked
 }
 {
  bone1Name=his3_hair_fr_A_c_r
  bone2Name=his3_hair_fr_A_b_r
  swing1LimitDegrees=2
  swing2LimitDegrees=2
  twistLimitDegrees=5
  twistMotion=Locked
 }
 {bone1Name=his3_hair_B_b_r, bone2Name=his3_hair_bk_b, swing1LimitDegrees=45, swing2LimitDegrees=45, twistLimitDegrees=45}
 {bone1Name=his3_hair_B_b_l, bone2Name=his3_hair_bk_b, swing1LimitDegrees=45, swing2LimitDegrees=45, twistLimitDegrees=45}
 {bone1Name=his3_hair_B_c_r, bone2Name=his3_hair_bk_c, swing1LimitDegrees=45, swing2LimitDegrees=45, twistLimitDegrees=45}
 {bone1Name=his3_hair_B_c_l, bone2Name=his3_hair_bk_c, swing1LimitDegrees=45, swing2LimitDegrees=45, twistLimitDegrees=45}
 {bone1Name=his3_hair_A_b_l, bone2Name=his3_hair_B_b_l, swing1LimitDegrees=45, swing2LimitDegrees=45, twistLimitDegrees=45}
 {bone1Name=his3_hair_B_c_r, bone2Name=his3_hair_bk_c, swing1LimitDegrees=45, swing2LimitDegrees=45, twistLimitDegrees=45}
 {bone1Name=his3_hair_A_c_l, bone2Name=his3_hair_B_c_l, swing1LimitDegrees=45, swing2LimitDegrees=45, twistLimitDegrees=45}
 {bone1Name=his3_hair_B_d_r, bone2Name=his3_hair_bk_c, swing1LimitDegrees=45, swing2LimitDegrees=45, twistLimitDegrees=45}
 {bone1Name=his3_hair_B_d_l, bone2Name=his3_hair_bk_d, swing1LimitDegrees=45, swing2LimitDegrees=45, twistLimitDegrees=45}
 {
  bone1Name=his3_hair_split_base_r
  bone2Name=his3_hair
  swing1LimitDegrees=45
  swing1Motion=Free
  swing2LimitDegrees=45
  swing2Motion=Free
  twistLimitDegrees=45
  twistMotion=Free
 }
 {bone1Name=his3_hair_split_r, bone2Name=his3_hair_split_base_r, swing1LimitDegrees=25, swing2LimitDegrees=25, twistLimitDegrees=5}
 {bone1Name=his3_hair_split_r, bone2Name=his3_hair_B_c_r, swing1LimitDegrees=45, swing2LimitDegrees=45, twistLimitDegrees=45}
 {bone1Name=head, bone2Name=his3_hair_A_b_r, swing1LimitDegrees=45, swing2LimitDegrees=45, twistLimitDegrees=45}
 {bone1Name=neck_01, bone2Name=his3_hair_A_c_r, swing1LimitDegrees=45, swing2LimitDegrees=45, twistLimitDegrees=45}
]
